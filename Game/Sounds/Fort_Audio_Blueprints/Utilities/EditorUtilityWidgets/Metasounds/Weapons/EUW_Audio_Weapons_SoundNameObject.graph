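# Draggable sound-name entry widget, with icon and text label, for an editor utility tool for weapon MetaSound audio.
# Left-button drag creates a B_Audio_Weapons_SoundNameDragDrop operation carrying item name, sound object and current action widget, using the widget itself as drag visual.
# Drag start, cancel and drop-onto-entry fire event dispatchers, leaving handling to listeners.
# Cancel also prints the action widget's preset text for debugging.
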EditorUtilityWidgetBlueprint /Game/Sounds/Fort_Audio_Blueprints/Utilities/EditorUtilityWidgets/Metasounds/Weapons/EUW_Audio_Weapons_SoundNameObject
graphs [
 EventGraph {nodes [
   Event OnDragCancelled!() -> (OutputDelegate, then, PointerEvent:"Pointer Event Structure (by ref)", Operation:Drag Drop Operation Object Reference) {is=title K2Node_Event}
   > Call On Drag Cancel!(
    execute:Exec
    self:EUW Audio Weapons Sound Name Object Object Reference
    SoundObject:Object Reference=Get SoundObjectRef() -> (SoundObjectRef:Object Reference)
    CurrentActionWidget:EUW Audio Weapons Action Name Widget Object Reference=Get Current Widget() -> (Current Widget:EUW Audio Weapons Action Name Widget Object Reference)
    SoundNameWidget:EUW Audio Weapons Sound Name Object Object Reference=Self-Reference()
   ) -> (then:Exec) {is=title K2Node_CallDelegate}
   n1 = "GetText (Text)"!(self:Text Object Reference=Get PresetText!(self:EUW Audio Weapons Action Name Widget Object Reference=Get Current Widget()) -> (PresetText:Text Object Reference)) -> (ReturnValue:Text)
   > PrintString(InString=Append(A="Drag Cancled: ", B="To String (Text)"(InText=@n1)))
   Event Construct()
  ]}
 OnDragCancel {nodes [OnDragCancel!() -> (
    then
    SoundObject:Object Reference
    CurrentActionWidget:EUW Audio Weapons Action Name Widget Object Reference
    SoundNameWidget:EUW Audio Weapons Sound Name Object Object Reference
   ) {is=title K2Node_FunctionEntry}]}
 OnDragDetached {nodes [OnDragDetached!() -> (
    then
    SoundObject:Object Reference
    CurrentNameWidget:EUW Audio Weapons Action Name Widget Object Reference
    ItemName:Text
    SoundNameWidget:EUW Audio Weapons Sound Name Object Object Reference
   ) {is=title K2Node_FunctionEntry}]}
 OnDragDetected {nodes [
   OnDragDetected!() -> (then, MyGeometry:Geometry Structure, PointerEvent:"Pointer Event Structure (by ref)") {is=title K2Node_FunctionEntry}
   > n1 = "Create Drag && Drop Operation"!(
    execute:Exec
    WorldContextObject:Object Reference
    Class:Drag Drop Operation Class Reference=&/B_Audio_Weapons_SoundNameDragDrop.B_Audio_Weapons_SoundNameDragDrop_C
    Tag:String
    Payload:Object Reference
    DefaultDragVisual:Widget Object Reference=Self-Reference()
    Pivot:EDragPivot Enum=CenterCenter
    Offset:Vector 2D Structure=(X=0,Y=0)
    ItemName:Text=Get ItemName() -> (ItemName:Text)
    CurrentWidget:EUW Audio Weapons Action Name Widget Object Reference=Get Current Widget()
    ObjectRef:Object Reference
    SoundObjectRef:Object Reference=Get SoundObjectRef()
   ) -> (then:Exec, ReturnValue:B Audio Weapons Sound Name Drag Drop Object Reference) {is=title K2Node_CreateDragDropOperation}
   > Call On Drag Detached!(
    execute:Exec
    self:EUW Audio Weapons Sound Name Object Object Reference
    SoundObject:Object Reference=Get SoundObjectRef()
    CurrentNameWidget:EUW Audio Weapons Action Name Widget Object Reference=Get Current Widget()
    ItemName:Text=Get ItemName()
    SoundNameWidget:EUW Audio Weapons Sound Name Object Object Reference=Self-Reference()
   ) -> (then:Exec) {is=title K2Node_CallDelegate}
   > Return Node!(execute, Operation:Drag Drop Operation Object Reference=@n1)
  ]}
 OnDraggedWidgetDropped {nodes [OnDraggedWidgetDropped!() -> (
    then
    DragDropRef:B Audio Weapons Sound Name Drag Drop Object Reference
    SoundNameWidget:EUW Audio Weapons Sound Name Object Object Reference
   ) {is=title K2Node_FunctionEntry}]}
 OnDrop {nodes [
   n1 = OnDrop!() -> (then, MyGeometry:Geometry Structure, PointerEvent:Pointer Event Structure, Operation:Drag Drop Operation Object Reference) {is=title K2Node_FunctionEntry}
   > n2 = Cast To B_Audio_Weapons_SoundNameDragDrop!(execute, Object=@n1.Operation) -> (then, CastFailed, AsB Audio Weapons Sound Name Drag Drop:B Audio Weapons Sound Name Drag Drop Object Reference) {is=title K2Node_DynamicCast}
   > Call On Dragged Widget Dropped!(
    execute:Exec
    self:EUW Audio Weapons Sound Name Object Object Reference
    DragDropRef:B Audio Weapons Sound Name Drag Drop Object Reference=@n2
    SoundNameWidget:EUW Audio Weapons Sound Name Object Object Reference=Self-Reference()
   ) -> (then:Exec) {is=title K2Node_CallDelegate}
   Return Node!(execute, ReturnValue:bool=false)
  ]}
 OnMouseButtonDown {nodes [
   n1 = OnMouseButtonDown!() -> (then, MyGeometry:Geometry Structure, MouseEvent:"Pointer Event Structure (by ref)") {is=title K2Node_FunctionEntry}
   > n2 = DetectDragIfPressed!(
    execute
    PointerEvent:"Pointer Event Structure (by ref)"=@n1.MouseEvent
    WidgetDetectingDrag:Widget Object Reference
    DragKey:Key Structure=LeftMouseButton
   ) -> (then, ReturnValue:Event Reply Structure)
   > Return Node(ReturnValue:Event Reply Structure=@n2)
  ]}
]
variables [
 Current Widget {type=EUW_Audio_Weapons_ActionNameWidget_C, default=null}
 CurrentScrollBox {type=EditorUtilityScrollBox, default=null}
 IsLooping {type=boolean}
 ItemName {type=string, default=""}
 New Scrotll Box {type=PanelWidget, default=null}
 OnDragCancel {default=()}
 OnDragDetached {default=()}
 OnDraggedWidgetDropped {default=()}
 ParentWidget {type=EUW_Audio_Weapons_SoundName_ListView_C, default=null}
 SoundObjectRef {type=Object, default=null}
 TypeText {type=string, default=""}
]
widgets {
 info {rootWidgetClass=&Overlay, widgetCount=7}
 widgets [
  Overlay_141 {widgetClassPath=&Overlay, widget=Overlay_141}
  Border_1 {
   widgetClassPath=&Border
   parent=Overlay_141
   widget=Border_1
   properties {
    background {
     bIsDynamicallyLoaded=false
     drawAs=Image
     imageType=NoImage
     mirroring=NoMirror
     resourceName=None
     resourceObject=null
     tiling=NoTile
     imageSize=<32 32>
     margin=<ltrb 0 0 0 0>
     outlineSettings {
      bUseBrushTransparency=false
      roundingType=HalfHeightRadius
      width=0
      color {colorUseRule=UseColor_Specified, specifiedColor=<rgba 0 0 0 0>}
      cornerRadii=<0 0 0 0>
     }
     tintColor {colorUseRule=UseColor_Specified, specifiedColor=<rgba 0.0208 0.0208 0.0208 1>}
     uVRegion {bIsValid=false, max=<0 0>, min=<0 0>}
    }
    padding=<ltrb 0 0 0 0>
   }
   slot {class=OverlaySlot, properties {horizontalAlignment=HAlign_Fill, verticalAlignment=VAlign_Fill, padding=<ltrb 4 2 4 2>}}
  }
  Border_0 {
   widgetClassPath=&Border
   parent=Border_1
   widget=Border_0
   properties {background {
     bIsDynamicallyLoaded=false
     drawAs=Border
     imageType=NoImage
     mirroring=NoMirror
     resourceName=None
     resourceObject=null
     tiling=NoTile
     imageSize=<32 32>
     margin=<ltrb 1 1 1 1>
     outlineSettings {
      bUseBrushTransparency=false
      roundingType=HalfHeightRadius
      width=0
      color {colorUseRule=UseColor_Specified, specifiedColor=<rgba 0 0 0 0>}
      cornerRadii=<0 0 0 0>
     }
     tintColor {colorUseRule=UseColor_Specified, specifiedColor=<rgba 0 0 0 1>}
     uVRegion {bIsValid=false, max=<0 0>, min=<0 0>}
    }}
   slot {class=BorderSlot, properties {padding=<ltrb 0 0 0 0>}}
  }
  HorizontalBox_128 {widgetClassPath=&HorizontalBox, parent=Border_0, widget=HorizontalBox_128, slot {class=BorderSlot}}
  SizeBox_0 {
   widgetClassPath=&SizeBox
   parent=HorizontalBox_128
   widget=SizeBox_0
   properties {bOverride_WidthOverride=true, widthOverride=61.4608, renderTransform {angle=0, scale=<0.65 1>, shear=<0 0>, translation=<0 0>}}
   slot {class=HorizontalBoxSlot, properties {horizontalAlignment=HAlign_Center, verticalAlignment=VAlign_Center}}
  }
  SoundIcon {
   widgetClassPath=&Image
   bIsVariable=true
   parent=SizeBox_0
   widget=SoundIcon
   properties {
    brush {
     bIsDynamicallyLoaded=false
     drawAs=Image
     imageType=FullColor
     mirroring=NoMirror
     resourceName=None
     resourceObject=/FMJamSystem/Textures/Icons/T_waveicon_Placeholder
     tiling=NoTile
     imageSize=<43 43>
     margin=<ltrb 0 0 0 0>
     outlineSettings {
      bUseBrushTransparency=false
      roundingType=HalfHeightRadius
      width=0
      color {colorUseRule=UseColor_Specified, specifiedColor=<rgba 0 0 0 0>}
      cornerRadii=<0 0 0 0>
     }
     tintColor {colorUseRule=UseColor_Specified, specifiedColor=<rgba 1 1 1 1>}
     uVRegion {bIsValid=false, max=<0 0>, min=<0 0>}
    }
    colorAndOpacity=<rgba 0.5029 0.5029 0.5029 1>
   }
   slot {class=SizeBoxSlot}
  }
  TextEntry {
   widgetClassPath=&TextBlock
   bIsVariable=true
   parent=HorizontalBox_128
   widget=TextEntry
   properties {
    text=Example
    colorAndOpacity {colorUseRule=UseColor_Specified, specifiedColor=<rgba 1 1 1 0.007>}
    font {
     bForceMonospaced=false
     bMaterialIsStencil=false
     fontMaterial=null
     fontObject=&Burbank
     letterSpacing=0
     monospacedWidth=1
     size=15
     skewAmount=0
     typefaceFontName=Burbank-Black
     outlineSettings {
      bApplyOutlineToDropShadows=false
      bMiteredCorners=false
      bSeparateFillAlpha=false
      outlineMaterial=null
      outlineSize=0
      outlineColor=<rgba 0 0 0 1>
     }
    }
    shadowColorAndOpacity=<rgba 0.8333 0.8333 0.8333 1>
    shadowOffset=<0.5 0.5>
   }
   slot {class=HorizontalBoxSlot, properties {verticalAlignment=VAlign_Center, padding=<ltrb 4 2 4 2>}}
  }
 ]
}
