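# Editor utility list widget for the weapon MetaSound audio tool, showing selected sound assets grouped under action-name rows.
# SetupWidgetActions builds a "Selected (Undefined)" row plus one per source wave name.
# Added sounds become entry widgets in the first row, then auto-move to rows whose action name maps to a substring of the sound's name.
# Drag-and-drop events re-parent entries between row scroll boxes; reset returns them to the default row.
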EditorUtilityWidgetBlueprint /Game/Sounds/Fort_Audio_Blueprints/Utilities/EditorUtilityWidgets/Metasounds/Weapons/EUW_Audio_Weapons_SoundName_ListView
graphs [
 AddSoundNameToActionWdiget {nodes [
   n1 = AddSoundNameToActionWdiget!() -> (then, ActionWidget:EUW Audio Weapons Action Name Widget Object Reference, SoundObject:Object Reference) {is=title K2Node_FunctionEntry}
   > n2 = Construct Object from Class!(
    execute
    Class=&/EUW_Audio_Weapons_SoundNameObject.EUW_Audio_Weapons_SoundNameObject_C
    self=@n1
    ItemName:Text="To Text (Object)"!(InObj:Object Reference=Get SoundObject() -> (SoundObject:Object Reference)) -> (ReturnValue:Text)
    Current Widget:EUW Audio Weapons Action Name Widget Object Reference=@n1
    SoundObjectRef:Object Reference=Get SoundObject()
    ParentWidget:EUW Audio Weapons Sound Name List View Object Reference=Self-Reference()
   ) -> (then, ReturnValue:EUW Audio Weapons Sound Name Object Object Reference)
   > Bind Event to On Drag Detached(self:EUW Audio Weapons Sound Name Object Object Reference=@n2, Delegate=Create Event()) {is=title K2Node_AddDelegate}
   > Bind Event to On Drag Cancel(self:EUW Audio Weapons Sound Name Object Object Reference=@n2, Delegate=Create Event()) {is=title K2Node_AddDelegate}
   > Add Unique(
    TargetArray:Array of EUW Audio Weapons Sound Name Object Object References=Get SoundObjectWidget() -> (SoundObjectWidget:Array of EUW Audio Weapons Sound Name Object Object References)
    NewItem:"EUW Audio Weapons Sound Name Object Object Reference (by ref)"=@n2
   )
   > Add Unique(
    TargetArray:Array of Object References=Get ObjectArray() -> (ObjectArray:Array of Object References)
    NewItem:"Object Reference (by ref)"=Get SoundObject()
   )
   > Add Child Widget!(
    execute
    self
    ObjectRef:Object Reference=Get SoundObject()
    ActionNameWidget:EUW Audio Weapons Action Name Widget Object Reference=Get ActionWidget() -> (ActionWidget:EUW Audio Weapons Action Name Widget Object Reference)
    SoundNameWidget:EUW Audio Weapons Sound Name Object Object Reference=@n2
   ) -> (then)
  ]}
 AttemptToAutomateObjectWidgetToActionWidgets {nodes [
   AttemptToAutomateObjectWidgetToActionWidgets!() -> (then, SoundObjectWidget:EUW Audio Weapons Sound Name Object Object Reference) {is=title K2Node_FunctionEntry}
   > n1 = For Each Loop(Array:Array of EUW Audio Weapons Action Name Widget Object References=Get WeaponActionNameWidgets() -> (WeaponActionNameWidgets:Array of EUW Audio Weapons Action Name Widget Object References)) -> (Array Element:EUW Audio Weapons Action Name Widget Object Reference)
   n2 = "To Text (Object)"(InObj=Get SoundObjectRef!(self:EUW Audio Weapons Sound Name Object Object Reference=Get SoundObjectWidget() -> (SoundObjectWidget:EUW Audio Weapons Sound Name Object Object Reference)) -> (SoundObjectRef:Object Reference))
   n3 = "GetText (Text)"!(self:Text Object Reference=Get PresetText!(self:EUW Audio Weapons Action Name Widget Object Reference=@n1) -> (PresetText:Text Object Reference)) -> (ReturnValue:Text)
   > Branch(Condition=IsProperString!(Substring:String="To String (Text)"(InText=@n3), SearchIn:String="To String (Text)"(InText=@n2)) -> (ReturnValue:bool))
   n4 = Get SoundObjectWidget()
   > MoveChildWidget!(
    execute
    self
    PreviousActionWidget:EUW Audio Weapons Action Name Widget Object Reference=Get Current Widget!(self:EUW Audio Weapons Sound Name Object Object Reference=@n4) -> (Current Widget:EUW Audio Weapons Action Name Widget Object Reference)
    NewActionWidget:EUW Audio Weapons Action Name Widget Object Reference=@n1
    ObjectRef:Object Reference=Get SoundObjectRef(self=@n4)
    SoundNameWidget:EUW Audio Weapons Sound Name Object Object Reference=@n4
   ) -> (then)
   > Set Current Widget!(
    execute
    Current Widget:EUW Audio Weapons Action Name Widget Object Reference=@n1
    self:EUW Audio Weapons Sound Name Object Object Reference=Get SoundObjectWidget()
   ) -> (then, Output_Get:EUW Audio Weapons Action Name Widget Object Reference)
  ]}
 EventGraph {nodes [
   n1 = OnDragCancel!() -> (
    OutputDelegate
    then
    SoundObject:Object Reference
    CurrentActionWidget:EUW Audio Weapons Action Name Widget Object Reference
    SoundNameWidget:EUW Audio Weapons Sound Name Object Object Reference
   ) {is=title K2Node_CustomEvent}
   > AddChild(
    self=Get ScrollBox!(self:EUW Audio Weapons Action Name Widget Object Reference=@n1.CurrentActionWidget) -> (ScrollBox:Editor Utility Scroll Box Object Reference)
    Content=@n1.SoundNameWidget
   )
   > Set Current Widget(Current Widget=@n1.CurrentActionWidget, self=@n1.SoundNameWidget)
   n2 = OnDragDropComplete!() -> (
    OutputDelegate
    then
    DragDropRef:B Audio Weapons Sound Name Drag Drop Object Reference
    ActionNameWdiget:EUW Audio Weapons Action Name Widget Object Reference
   ) {is=title K2Node_CustomEvent}
   n3 = Get Sound Name Widget() -> (Sound Name Widget:EUW Audio Weapons Sound Name Object Object Reference)
   > AddChild(self=Get ScrollBox(self=@n2.ActionNameWdiget), Content=@n3)
   > Set Current Widget(Current Widget=@n2.ActionNameWdiget, self=@n3)
   n4 = Add Child Widget!() -> (
    OutputDelegate
    then
    ObjectRef:Object Reference
    ActionNameWidget:EUW Audio Weapons Action Name Widget Object Reference
    SoundNameWidget:EUW Audio Weapons Sound Name Object Object Reference
   ) {is=title K2Node_CustomEvent}
   > n5 = Set Current Widget(Current Widget=@n4.ActionNameWidget, self=@n4.SoundNameWidget)
   > AddChild(self=Get ScrollBox(self=@n5), Content=@n4.SoundNameWidget)
   n6 = MoveChildWidget!() -> (
    OutputDelegate
    then
    PreviousActionWidget:EUW Audio Weapons Action Name Widget Object Reference
    NewActionWidget:EUW Audio Weapons Action Name Widget Object Reference
    ObjectRef:Object Reference
    SoundNameWidget:EUW Audio Weapons Sound Name Object Object Reference
   ) {is=title K2Node_CustomEvent}
   > Branch(Condition="Not Equal (Object)"(A:Object Reference=@n6.PreviousActionWidget, B:Object Reference=@n6.NewActionWidget) -> (ReturnValue:bool=~) {is=title K2Node_PromotableOperator})
   > RemoveChild(self=Get ScrollBox(self=@n6.PreviousActionWidget), Content=@n6.SoundNameWidget)
   > AddChild(self=Get ScrollBox(self=@n6.NewActionWidget), Content=@n6.SoundNameWidget)
   > Set Current Widget(Current Widget=@n6.NewActionWidget, self=@n6.SoundNameWidget)
   n7 = Remove Child Widget!() -> (
    OutputDelegate
    then
    ActionNameWidget:EUW Audio Weapons Action Name Widget Object Reference
    SoundNameWdiget:"EUW Audio Weapons Sound Name Object Object Reference (by ref)"
   ) {is=title K2Node_CustomEvent}
   > RemoveChild(self=Get ScrollBox(self=@n7.ActionNameWidget), Content=@n7.SoundNameWdiget)
   ResetActionWidgets() {is=title K2Node_CustomEvent}
   > n8 = Sequence()
   > n9 = For Each Loop(Array=Get WeaponActionNameWidgets())
   Clear(execute=@n9.Completed, TargetArray:Array of EUW Audio Weapons Action Name Widget Object References=Get WeaponActionNameWidgets())
   > SetupWidgetActions!(execute, self, ActionName:String="Selected (Undefined)") -> (then)
   RemoveChild(execute=@n9.LoopBody, self=Get MainBox() -> (MainBox:Editor Utility Scroll Box Object Reference), Content=@n9)
   n10 = For Each Loop(
    Exec=@n8.then_1
    Array:Array of EUW Audio Weapons Sound Name Object Object References=Get SoundObjectWidget() -> (SoundObjectWidget:Array of EUW Audio Weapons Sound Name Object Object References)
   ) -> (Array Element:EUW Audio Weapons Sound Name Object Object Reference)
   > MoveChildWidget(
    PreviousActionWidget=Get Current Widget(self=@n10)
    NewActionWidget="Get (a copy)"(Array:Array of EUW Audio Weapons Action Name Widget Object References=Get WeaponActionNameWidgets()) -> (Output:EUW Audio Weapons Action Name Widget Object Reference)
    ObjectRef=Get SoundObjectRef(self=@n10)
    SoundNameWidget=@n10
   ) {is=title K2Node_CallFunction}
   n11 = OnDraggedDetected!() -> (
    OutputDelegate
    then
    SoundObject:Object Reference
    ActionNameWidget:EUW Audio Weapons Action Name Widget Object Reference
    ItemName:Text
    SoundNameWidget:EUW Audio Weapons Sound Name Object Object Reference
   ) {is=title K2Node_CustomEvent}
   > Remove Child Widget!(
    execute
    self
    ActionNameWidget:EUW Audio Weapons Action Name Widget Object Reference=@n11.ActionNameWidget
    SoundNameWdiget:"EUW Audio Weapons Sound Name Object Object Reference (by ref)"=@n11.SoundNameWidget
   ) -> (then) {is=title K2Node_CallFunction}
   > Set Sound Name Widget(Sound Name Widget:EUW Audio Weapons Sound Name Object Object Reference=@n11.SoundNameWidget) -> (Output_Get:EUW Audio Weapons Sound Name Object Object Reference)
   > Set Current Widget(Current Widget=@n11.ActionNameWidget, self=@n11.SoundNameWidget)
   n12 = OnDragDrop!() -> (
    OutputDelegate
    then
    DragDropRef:B Audio Weapons Sound Name Drag Drop Object Reference
    ActionName:EUW Audio Weapons Action Name Widget Object Reference
   ) {is=title K2Node_CustomEvent}
   > n13 = For Each Loop(Array=Get SoundObjectWidget())
   n14 = Get ObjectRef!(self:B Audio Weapons Sound Name Drag Drop Object Reference=@n12.DragDropRef) -> (ObjectRef:Object Reference)
   n15 = "To Text (Object)"(InObj=@n14)
   n16 = "To Text (Object)"(InObj=@n13)
   > Branch(Condition=IsProperString(Substring="To String (Text)"(InText=@n16), SearchIn="To String (Text)"(InText=@n15)))
   > MoveChildWidget(
    PreviousActionWidget=Get CurrentWidget!(self:B Audio Weapons Sound Name Drag Drop Object Reference=@n12.DragDropRef) -> (CurrentWidget:EUW Audio Weapons Action Name Widget Object Reference)
    NewActionWidget=@n12.ActionName
    ObjectRef=@n14
    SoundNameWidget=@n13
   )
   n17 = GetSoundNameReferences!() -> (OutputDelegate, then, Skip:bool=~, SourceWaveNames:Array of Strings, SoundGroupType:String) {is=title K2Node_CustomEvent}
   > n18 = Branch(Condition=@n17.Skip)
   n19 = For Each Loop(Exec=@n18.else, Array:Array of Strings=@n17.SourceWaveNames) -> (Array Element:String)
   Set SoundGroupType(execute=@n19.Completed, SoundGroupType:String=@n17.SoundGroupType) -> (Output_Get:String)
   SetupWidgetActions(execute=@n19.LoopBody, ActionName=@n19)
   n20 = AddEntryUpdate!() -> (OutputDelegate, then, Object:Object Reference, AssetIndex:"Integer (by ref)") {is=title K2Node_CustomEvent}
   > Add Unique(TargetArray:Array of Integers=Get AssetIndexes() -> (AssetIndexes:Array of Integers), NewItem:"Integer (by ref)"=@n20.AssetIndex)
   > AddSoundNameToActionWdiget!(
    execute
    self
    ActionWidget:EUW Audio Weapons Action Name Widget Object Reference="Get (a copy)"(Array=Get WeaponActionNameWidgets())
    SoundObject:Object Reference=@n20.Object
   ) -> (then) {is=title K2Node_CallFunction}
   > n21 = For Each Loop(Array:Array of EUW Audio Weapons Sound Name Object Object References=Get SoundObjectWidget()) -> (Array Element:EUW Audio Weapons Sound Name Object Object Reference)
   > AttemptToAutomateObjectWidgetToActionWidgets!(execute, self, SoundObjectWidget:EUW Audio Weapons Sound Name Object Object Reference=@n21) -> (then) {is=title K2Node_CallFunction}
   n22 = RemoveEntryUpdate!() -> (OutputDelegate, then, Object:Object Reference, AssetIndex:"Integer (by ref)") {is=title K2Node_CustomEvent}
   n23 = "Get (a copy)"(
    Array:Array of EUW Audio Weapons Sound Name Object Object References=Get SoundObjectWidget()
    Dimension 1="Get (a copy)"(Array:Array of Integers=Get AssetIndexes(), Dimension 1=@n22.AssetIndex) -> (Output:int)
   ) -> (Output:EUW Audio Weapons Sound Name Object Object Reference)
   > Remove Child Widget(ActionNameWidget=Get Current Widget(self=@n23), SoundNameWdiget=@n23)
   > Remove Item!(
    execute
    TargetArray:Array of EUW Audio Weapons Sound Name Object Object References=Get SoundObjectWidget()
    Item:"EUW Audio Weapons Sound Name Object Object Reference (by ref)"=@n23
   ) -> (then, ReturnValue:bool)
   Event PreConstruct()
   > SetupWidgetActions(ActionName="Selected (Undefined)")
   Get PresetText(self=@n11.ActionNameWidget)
  ]}
 IsInvalidText {nodes [
   IsInvalidText() {is=title K2Node_FunctionEntry}
   n1 = IsEmpty(InString=Get SoundGroupType() -> (SoundGroupType:String))
   n2 = Is Empty!(TargetArray:Array of User Widget Object References=Get InvalidTextArray() -> (InvalidTextArray:Array of User Widget Object References)) -> (ReturnValue:bool)
   > Return Node!(execute, ReturnValue:bool=AND Boolean(A=NOT Boolean(A=@n2), B=NOT Boolean(A=@n1)))
  ]}
 IsProperString {nodes [
   n1 = IsProperString!() -> (then, Substring:String, SearchIn:String) {is=title K2Node_FunctionEntry}
   n2 = Find(TargetMap=Get TMapStringToStringInput() -> (TMapStringToStringInput:Map of Strings to Strings), Key=@n1)
   > n3 = Branch(Condition=OR Boolean!(A=IsEmpty(InString=@n1), B=IsEmpty(InString=@n1.SearchIn), C:bool=IsEmpty(InString=@n2)) -> (ReturnValue))
   > Return Node(ReturnValue=false)
   Return Node(execute=@n3.else, ReturnValue=Contains(SearchIn=@n1.SearchIn, Substring=@n2))
  ]}
 SetOutputText {nodes [
   n1 = SetOutputText!() -> (then, Text:Text) {is=title K2Node_FunctionEntry}
   > Set Text(Text=@n1, self=Get OutputText() -> (OutputText:Text Object Reference))
  ]}
 SetStringToStringMap {nodes [
   n1 = SetStringToStringMap!() -> (then, TMapStringToStringInput:Map of Strings to Strings) {is=title K2Node_FunctionEntry}
   > Clear!(execute, TargetMap:Map of Strings to Strings=Get TMapStringToStringInput()) -> (then) {is=title K2Node_CallFunction}
   > Set TMapStringToStringInput(TMapStringToStringInput:Map of Strings to Strings=@n1) -> (Output_Get:Map of Strings to Strings)
  ]}
 SetupWidgetActions {nodes [
   SetupWidgetActions!() -> (then, ActionName:String) {is=title K2Node_FunctionEntry}
   > n1 = Construct Object from Class!(
    execute
    Class=&/EUW_Audio_Weapons_ActionNameWidget.EUW_Audio_Weapons_ActionNameWidget_C
    self=Get MainBox()
   ) -> (then, ReturnValue:EUW Audio Weapons Action Name Widget Object Reference)
   > Add Unique(
    TargetArray:Array of EUW Audio Weapons Action Name Widget Object References=Get WeaponActionNameWidgets()
    NewItem:"EUW Audio Weapons Action Name Widget Object Reference (by ref)"=@n1
   )
   > AddChild(self=Get MainBox(), Content=@n1)
   > "SetText (Text)"(self=Get PresetText(self=@n1), InText="To Text (String)"(InString=Get ActionName() -> (ActionName:String)))
   > SetAlwaysShowScrollbar!(execute, self:Scroll Box Object Reference=Get ScrollBox(self=@n1), NewAlwaysShowScrollbar:bool=true) -> (then)
   > Bind Event to On Drag Drop Complete(self:EUW Audio Weapons Action Name Widget Object Reference=@n1, Delegate=Create Event()) {is=title K2Node_AddDelegate}
  ]}
]
variables [
 Action Name Widget {type=EUW_Audio_Weapons_ActionNameWidget_C, default=null}
 AssetIndexes {type="array<integer>"}
 InvalidTextArray {type="array<UserWidget>", default []}
 ObjectArray {type="array<Object>", default []}
 Sound Name Widget {type=EUW_Audio_Weapons_SoundNameObject_C, default=null}
 SoundGroupType {type=string, default=""}
 SoundObjectWidget {type="array<EUW_Audio_Weapons_SoundNameObject_C>", default []}
 TMapStringToStringInput {type="map<string>", default {}}
 WeaponActionNameWidgets {type="array<EUW_Audio_Weapons_ActionNameWidget_C>", default []}
]
widgets {
 info {rootWidgetClass=&CanvasPanel, widgetCount=7}
 widgets [
  CanvasPanel_0 {widgetClassPath=&CanvasPanel, widget=CanvasPanel_0}
  VerticalBox_59 {
   widgetClassPath=&VerticalBox
   parent=CanvasPanel_0
   widget=VerticalBox_59
   slot {class=CanvasPanelSlot, properties {layoutData {alignment=<0 0>, anchors {maximum=<0 0>, minimum=<0 0>}, offsets=<ltrb 0 0 600 473.8739>}}}
  }
  Border_193 {
   widgetClassPath=&Border
   parent=VerticalBox_59
   widget=Border_193
   properties {
    horizontalAlignment=HAlign_Center
    background {
     bIsDynamicallyLoaded=false
     drawAs=RoundedBox
     imageType=NoImage
     mirroring=NoMirror
     resourceName=None
     resourceObject=null
     tiling=NoTile
     imageSize=<32 32>
     margin=<ltrb 0 0 0 0>
     outlineSettings {
      bUseBrushTransparency=false
      roundingType=FixedRadius
      width=1
      color {colorUseRule=UseColor_Specified, specifiedColor=<rgba 0.1198 0.1198 0.1198 0.5>}
      cornerRadii=<0 0 0 0>
     }
     tintColor {colorUseRule=UseColor_Specified, specifiedColor=<rgba 0.0423 0.0423 0.0423 1>}
     uVRegion {bIsValid=false, max=<0 0>, min=<0 0>}
    }
    padding=<ltrb 4 4 4 4>
   }
   slot {class=VerticalBoxSlot}
  }
  SoundListText {
   widgetClassPath=&TextBlock
   bIsVariable=true
   parent=Border_193
   widget=SoundListText
   properties {
    text=Selected Sounds List
    font {
     bForceMonospaced=false
     bMaterialIsStencil=false
     fontMaterial=null
     fontObject=&Burbank
     letterSpacing=0
     monospacedWidth=1
     size=15
     skewAmount=0
     typefaceFontName=Burbank-Black
     outlineSettings {
      bApplyOutlineToDropShadows=false
      bMiteredCorners=false
      bSeparateFillAlpha=false
      outlineMaterial=null
      outlineSize=0
      outlineColor=<rgba 0 0 0 1>
     }
    }
    shadowColorAndOpacity=<rgba 0 0 0 1>
   }
   slot {class=BorderSlot, properties {horizontalAlignment=HAlign_Center, padding=<ltrb 4 4 4 4>}}
  }
  EditorUtilityScrollBox_103 {
   widgetClassPath=&EditorUtilityScrollBox
   parent=VerticalBox_59
   widget=EditorUtilityScrollBox_103
   properties {
    clipping=Inherit
    widgetBarStyle {
     thickness=8
     draggedThumbImage {
      bIsDynamicallyLoaded=false
      drawAs=RoundedBox
      imageType=NoImage
      mirroring=NoMirror
      resourceName=None
      resourceObject=null
      tiling=NoTile
      imageSize=<0 0>
      margin=<ltrb 0 0 0 0>
      outlineSettings {
       bUseBrushTransparency=false
       roundingType=FixedRadius
       width=0
       color {colorUseRule=UseColor_Specified, specifiedColor=<rgba 0 0 0 0>}
       cornerRadii=<4 4 4 4>
      }
      tintColor {colorUseRule=UseColor_Specified, specifiedColor=<rgba 0.2159 0.2159 0.2159 1>}
      uVRegion {bIsValid=false, max=<0 0>, min=<0 0>}
     }
     horizontalBackgroundImage {
      bIsDynamicallyLoaded=false
      drawAs=NoDrawType
      imageType=NoImage
      mirroring=NoMirror
      resourceName=None
      resourceObject=null
      tiling=NoTile
      imageSize=<0 0>
      margin=<ltrb 0 0 0 0>
      outlineSettings {
       bUseBrushTransparency=false
       roundingType=HalfHeightRadius
       width=0
       color {colorUseRule=UseColor_Specified, specifiedColor=<rgba 0 0 0 0>}
       cornerRadii=<0 0 0 0>
      }
      tintColor {colorUseRule=UseColor_Specified, specifiedColor=<rgba 1 1 1 1>}
      uVRegion {bIsValid=false, max=<0 0>, min=<0 0>}
     }
     horizontalBottomSlotImage {
      bIsDynamicallyLoaded=false
      drawAs=NoDrawType
      imageType=NoImage
      mirroring=NoMirror
      resourceName=None
      resourceObject=null
      tiling=NoTile
      imageSize=<0 0>
      margin=<ltrb 0 0 0 0>
      outlineSettings {
       bUseBrushTransparency=false
       roundingType=HalfHeightRadius
       width=0
       color {colorUseRule=UseColor_Specified, specifiedColor=<rgba 0 0 0 0>}
       cornerRadii=<0 0 0 0>
      }
      tintColor {colorUseRule=UseColor_Specified, specifiedColor=<rgba 1 1 1 1>}
      uVRegion {bIsValid=false, max=<0 0>, min=<0 0>}
     }
     horizontalTopSlotImage {
      bIsDynamicallyLoaded=false
      drawAs=NoDrawType
      imageType=NoImage
      mirroring=NoMirror
      resourceName=None
      resourceObject=null
      tiling=NoTile
      imageSize=<0 0>
      margin=<ltrb 0 0 0 0>
      outlineSettings {
       bUseBrushTransparency=false
       roundingType=HalfHeightRadius
       width=0
       color {colorUseRule=UseColor_Specified, specifiedColor=<rgba 0 0 0 0>}
       cornerRadii=<0 0 0 0>
      }
      tintColor {colorUseRule=UseColor_Specified, specifiedColor=<rgba 1 1 1 1>}
      uVRegion {bIsValid=false, max=<0 0>, min=<0 0>}
     }
     hoveredThumbImage {
      bIsDynamicallyLoaded=false
      drawAs=RoundedBox
      imageType=NoImage
      mirroring=NoMirror
      resourceName=None
      resourceObject=null
      tiling=NoTile
      imageSize=<0 0>
      margin=<ltrb 0 0 0 0>
      outlineSettings {
       bUseBrushTransparency=false
       roundingType=FixedRadius
       width=0
       color {colorUseRule=UseColor_Specified, specifiedColor=<rgba 0 0 0 0>}
       cornerRadii=<4 4 4 4>
      }
      tintColor {colorUseRule=UseColor_Specified, specifiedColor=<rgba 0.2159 0.2159 0.2159 1>}
      uVRegion {bIsValid=false, max=<0 0>, min=<0 0>}
     }
     normalThumbImage {
      bIsDynamicallyLoaded=false
      drawAs=RoundedBox
      imageType=NoImage
      mirroring=NoMirror
      resourceName=None
      resourceObject=null
      tiling=NoTile
      imageSize=<0 0>
      margin=<ltrb 0 0 0 0>
      outlineSettings {
       bUseBrushTransparency=false
       roundingType=FixedRadius
       width=0
       color {colorUseRule=UseColor_Specified, specifiedColor=<rgba 0 0 0 0>}
       cornerRadii=<4 4 4 4>
      }
      tintColor {colorUseRule=UseColor_Specified, specifiedColor=<rgba 0.0953 0.0953 0.0953 0>}
      uVRegion {bIsValid=false, max=<0 0>, min=<0 0>}
     }
     verticalBackgroundImage {
      bIsDynamicallyLoaded=false
      drawAs=NoDrawType
      imageType=NoImage
      mirroring=NoMirror
      resourceName=None
      resourceObject=null
      tiling=NoTile
      imageSize=<0 0>
      margin=<ltrb 0 0 0 0>
      outlineSettings {
       bUseBrushTransparency=false
       roundingType=HalfHeightRadius
       width=0
       color {colorUseRule=UseColor_Specified, specifiedColor=<rgba 0 0 0 0>}
       cornerRadii=<0 0 0 0>
      }
      tintColor {colorUseRule=UseColor_Specified, specifiedColor=<rgba 1 1 1 1>}
      uVRegion {bIsValid=false, max=<0 0>, min=<0 0>}
     }
     verticalBottomSlotImage {
      bIsDynamicallyLoaded=false
      drawAs=NoDrawType
      imageType=NoImage
      mirroring=NoMirror
      resourceName=None
      resourceObject=null
      tiling=NoTile
      imageSize=<0 0>
      margin=<ltrb 0 0 0 0>
      outlineSettings {
       bUseBrushTransparency=false
       roundingType=HalfHeightRadius
       width=0
       color {colorUseRule=UseColor_Specified, specifiedColor=<rgba 0 0 0 0>}
       cornerRadii=<0 0 0 0>
      }
      tintColor {colorUseRule=UseColor_Specified, specifiedColor=<rgba 1 1 1 1>}
      uVRegion {bIsValid=false, max=<0 0>, min=<0 0>}
     }
     verticalTopSlotImage {
      bIsDynamicallyLoaded=false
      drawAs=NoDrawType
      imageType=NoImage
      mirroring=NoMirror
      resourceName=None
      resourceObject=null
      tiling=NoTile
      imageSize=<0 0>
      margin=<ltrb 0 0 0 0>
      outlineSettings {
       bUseBrushTransparency=false
       roundingType=HalfHeightRadius
       width=0
       color {colorUseRule=UseColor_Specified, specifiedColor=<rgba 0 0 0 0>}
       cornerRadii=<0 0 0 0>
      }
      tintColor {colorUseRule=UseColor_Specified, specifiedColor=<rgba 1 1 1 1>}
      uVRegion {bIsValid=false, max=<0 0>, min=<0 0>}
     }
    }
   }
   slot {class=VerticalBoxSlot, properties {padding=<ltrb 2 2 2 2>}}
  }
  OutputText {
   widgetClassPath=&TextBlock
   bIsVariable=true
   parent=EditorUtilityScrollBox_103
   widget=OutputText
   properties {
    visibility=SelfHitTestInvisible
    font {
     bForceMonospaced=false
     bMaterialIsStencil=false
     fontMaterial=null
     fontObject=/LauncherSocial/BrutalType
     letterSpacing=0
     monospacedWidth=1
     size=15
     skewAmount=0
     typefaceFontName=Regular
     outlineSettings {
      bApplyOutlineToDropShadows=false
      bMiteredCorners=false
      bSeparateFillAlpha=false
      outlineMaterial=null
      outlineSize=0
      outlineColor=<rgba 0 0 0 1>
     }
    }
    shadowColorAndOpacity=<rgba 0 0 0 1>
    shadowOffset=<2 1>
   }
   slot {class=ScrollBoxSlot}
  }
  MainBox {
   widgetClassPath=&EditorUtilityScrollBox
   bIsVariable=true
   parent=VerticalBox_59
   widget=MainBox
   properties {alwaysShowScrollbar=true, alwaysShowScrollbarTrack=true}
   slot {class=VerticalBoxSlot, properties {size {sizeRule=Fill, value=1}}}
  }
 ]
}
